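; Cliff static mesh actor Blueprint for the Hera CosmicThunder terrain set that feeds a scale-compensated grid value to its material.
; Its construction script divides the GridScale variable by the largest absolute component of the actor's scale, clamped to a tiny minimum to avoid division by zero.
; The result is written to the mesh component's default custom primitive data at index 4, avoiding per-instance dynamic materials.
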(fn ConstructionScript
  (Rendering|Material|SetDefaultCustomPrimitiveDataFloat (Variables|StaticMeshActor|GetStaticMeshComponent) 4 (/ (Variables|Default|GetGridScale) (Math|Float|Max(Float) (Math|Vector|GetMaxElement (Math|Vector|VectorGetAbs (Transformation|GetActorScale3D))) 1e-05))))
