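; Parent Blueprint for Hera CosmicThunder cliff static mesh actors, feeding per-instance grid and stripe pattern values to the mesh's material.
; The construction script writes grid scale and offset to custom primitive data from index 4, dividing the larger horizontal and the vertical actor scale by GridScale.
; It writes stripe scale and offset from index 8, multiplying StripeScale by actor scale projected onto StripeScaleAxis.
; The event graph holds only disabled stubs.
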(event EventBeginPlay
  (|CallParentFunction) ; This node is disabled and will not be called.\nDrag off pins to build functionality.
)

(event Collision|EventActorBeginOverlap
  (params (OtherActor :type Actor%20Object%20Reference))
  (|CallParentFunction OtherActor) ; This node is disabled and will not be called.\nDrag off pins to build functionality.
)

(event EventTick
  (params (DeltaSeconds :type Float%20%28single-precision%29))
  (|CallParentFunction DeltaSeconds) ; This node is disabled and will not be called.\nDrag off pins to build functionality.
)

(fn ConstructionScript
  (bind (_x _y) (Math|Vector2D|BreakVector2D (Variables|Default|GetGridOffset)))
  (bind (_x_2 _y_2) (Math|Vector2D|BreakVector2D (Variables|Default|GetGridScale)))
  (|CallParentFunction)
  (Rendering|Material|SetDefaultCustomPrimitiveDataVector4 (Variables|StaticMeshActor|GetStaticMeshComponent) 4 (Math|Vector4|MakeVector4 (/ (Math|Float|Max(Float) (Math|Float|Max(Float) (.x (Math|Vector|VectorGetAbs (Transformation|GetActorScale3D))) (.y (Math|Vector|VectorGetAbs (Transformation|GetActorScale3D)))) 1e-05) _x_2) (/ (Math|Float|Max(Float) (.z (Math|Vector|VectorGetAbs (Transformation|GetActorScale3D))) 1e-05) _y_2) _x _y))
  (Rendering|Material|SetDefaultCustomPrimitiveDataVector2 (Variables|StaticMeshActor|GetStaticMeshComponent) 8 (Math|Vector2D|MakeVector2D (* (Variables|Default|GetStripeScale) (Math|Float|Max(Float) (Math|Vector|DotProduct (Math|Vector|VectorGetAbs (Transformation|GetActorScale3D)) (Variables|Default|GetStripeScaleAxis)) 1e-05)) (Variables|Default|GetStripeOffset))))
